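# Physics asset for the female medium Dark Elf backpack cosmetic, covering its dangling crystal bones.
# Two sphere bodies sit on dyn_crystal and dyn_crystal_MSTR: a larger one offset along the crystal bone's X axis and a small one at the master bone's origin.
# A single constraint joins them, with narrow, equal swing limit angles and a zero twist limit.
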PhysicsAsset /BRCosmetics/Accessories/FORT_Backpacks/Backpack_F_MED_Dark_Elf/Meshes/F_MED_Dark_Elf_Pack_Physics
bodies {dyn_crystal [{shapeType=Sphere, radius=2.2142, center=<3.4908 0 0>}], dyn_crystal_MSTR [{shapeType=Sphere, radius=0.505}]}
constraints [{bone1Name=dyn_crystal, bone2Name=dyn_crystal_MSTR, swing1LimitDegrees=20, swing2LimitDegrees=20, twistLimitDegrees=0}]
